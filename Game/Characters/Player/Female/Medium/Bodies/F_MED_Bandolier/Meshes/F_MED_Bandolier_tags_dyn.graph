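# Physics asset for the dangling tags, chain and bullets on the medium female player's Bandolier body.
# Three chain bodies link spine_05 to the tags body, from which three bullet capsules branch.
# The chain root swings and twists freely on the spine, while later joints have tight twist and narrower swing limits.
# Three capsules on spine_05 approximate the upper torso as a collision shape.
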PhysicsAsset /Game/Characters/Player/Female/Medium/Bodies/F_MED_Bandolier/Meshes/F_MED_Bandolier_tags_dyn
bodies {
 dyn_bullet_01 [{shapeType=Capsule, length=2, radius=0.8, center=<3 0 0>, rotation=<rot 90 0 0>}]
 dyn_bullet_02 [{shapeType=Capsule, length=2, radius=0.9, center=<3 0 0>, rotation=<rot 90 0 0>}]
 dyn_bullet_03 [{shapeType=Capsule, length=2, radius=0.85, center=<3 0 0>, rotation=<rot 90 0 0>}]
 dyn_chains_01 [{shapeType=Capsule, length=1, radius=1}]
 dyn_chains_02 [{shapeType=Box, extentX=6, extentY=12, extentZ=0.5, center=<3 0 0>, rotation=<rot -5 0 0>}]
 dyn_chains_03 [{shapeType=Box, extentX=6, extentY=9, extentZ=0.5, center=<3 0 0>}]
 dyn_tags [{shapeType=Box, extentX=4.883, extentY=5.5442, extentZ=1, center=<2.6837 0 -0.0001>}]
 spine_05 [
  {shapeType=Capsule, length=9.3014, radius=10.0445, center=<-0.0786 -2.0442 0>}
  {shapeType=Capsule, length=9.3014, radius=10.0445, center=<-7.7507 2.7012 0>}
  {shapeType=Capsule, length=10.1884, radius=3.7436, center=<-1.0876 9.1145 7.6711>, rotation=<rot -53.2607 8.2294 -46.317>}
 ]
}
constraints [
 {
  bone1Name=dyn_chains_01
  bone2Name=spine_05
  swing1LimitDegrees=45
  swing1Motion=Free
  swing2LimitDegrees=45
  swing2Motion=Free
  twistLimitDegrees=45
  twistMotion=Free
 }
 {bone1Name=dyn_chains_02, bone2Name=dyn_chains_01, swing1LimitDegrees=30, swing2LimitDegrees=10, twistLimitDegrees=2}
 {bone1Name=dyn_chains_03, bone2Name=dyn_chains_02, swing1LimitDegrees=30, swing2LimitDegrees=10, twistLimitDegrees=2}
 {bone1Name=dyn_tags, bone2Name=dyn_chains_03, swing1LimitDegrees=30, swing2LimitDegrees=10, twistLimitDegrees=2}
 {bone1Name=dyn_bullet_01, bone2Name=dyn_tags, swing1LimitDegrees=30, swing2LimitDegrees=15, twistLimitDegrees=2}
 {bone1Name=dyn_bullet_02, bone2Name=dyn_tags, swing1LimitDegrees=30, swing2LimitDegrees=15, twistLimitDegrees=2}
 {bone1Name=dyn_bullet_03, bone2Name=dyn_tags, swing1LimitDegrees=30, swing2LimitDegrees=15, twistLimitDegrees=2}
]
