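# Post-process style AnimBlueprint for the female medium Bandolier body part, adding secondary motion to gear on a pose copied from the attached parent mesh.
# A chain constrains the right kneepad, simulates RigidBody and bandana AnimDynamics, aims the ammo-belt corrective at upperarm_r and tilts the hip magazine.
# The chains offset has zero alpha; RBF Pose Driver, tags AnimDynamics and spare kneepad constraints are unconnected.
# The update event computes thigh_l_alpha, which the AnimGraph never reads.
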AnimBlueprint /Game/Characters/Player/Female/Medium/Bodies/F_MED_Bandolier/Meshes/F_MED_Bandolier_Skeleton_AnimBlueprint
target_skeleton=&/F_MED_Bandolier_Skeleton
graphs [
 AnimGraph {nodes [
   Anim Dynamics - Bone: dyn_tags(Alpha=0.7, ExternalForce="0,0,0") {
    angularBiasOverride=1
    angularDampingOverride=0.9999
    bGravityOverrideInSimSpace=true
    bOverrideAngularDamping=true
    gravityScale=0.35
    lODThreshold=1
    alphaBoolBlend {blendOption=Linear}
    boundBone {boneName=dyn_tags}
    physicsBodyDefinitions [{
      sphereCollisionRadius=0
      boundBone {boneName=dyn_tags}
      boxExtents=<5 2 10>
      constraintSetup {angularTargetAxis=AxisZ, twistAxis=AxisZ, angularLimitsMax=<40 45 0>, angularLimitsMin=<-20 -45 0>, angularTarget=<0 0 0>}
      localJointOffset=<0 0 -2>
     }]
   }
   "Constraint - kneepad_r (PP)"!(ComponentPose, Alpha=1, ConstraintWeights_0=0.3439, ConstraintWeights_1:float=0.6561) -> (Pose) {
    alphaBoolBlend {blendOption=Linear}
    boneToModify {boneName=kneepad_r}
    constraintSetup [{targetBone {boneName=thigh_r}}, {targetBone {boneName=calf_r}}]
    constraintWeights [1,1]
   }
   "Constraint - kneepad_r (PP)"(Alpha=1, ConstraintWeights_0=0.7185, ConstraintWeights_1=0.2815) {
    alphaBoolBlend {blendOption=Linear}
    boneToModify {boneName=kneepad_r}
    constraintSetup [{targetBone {boneName=thigh_r}}, {targetBone {boneName=calf_r}}]
    constraintWeights [1,1]
   }
   Pose Driver - Source: upperarm_r!(SourcePose:Pose Link Structure) -> (Pose:Pose Link Structure) {
    is=title AnimGraphNode_PoseDriver
    poseAsset=&/F_MED_Bandolier_RBF_PoseAsset
    onlyDriveBones [{boneName=dyn_ammobelt_corrective}]
    poseTargets [
     {
      bApplyCustomCurve=false
      bIsHidden=false
      distanceMethod=DefaultMethod
      drivenName=Pose_0
      functionType=DefaultFunction
      targetScale=0.4084
      boneTransforms [{targetRotation=<rot -58.1985 2.4027 1.3901>, targetTranslation=<-9.8658 1.2136 0.4104>}]
      customCurve {defaultValue=3.4028234663852886e+38, postInfinityExtrap=RCCE_Constant, preInfinityExtrap=RCCE_Constant, keys []}
      targetRotation=<rot 0 0 0>
     }
     {
      bApplyCustomCurve=false
      bIsHidden=false
      distanceMethod=DefaultMethod
      drivenName=Pose_1
      functionType=DefaultFunction
      targetScale=0.2583
      boneTransforms [{targetRotation=<rot -69.5594 100.5993 -48.4106>, targetTranslation=<-9.8658 1.2136 0.4104>}]
      customCurve {defaultValue=3.4028234663852886e+38, postInfinityExtrap=RCCE_Constant, preInfinityExtrap=RCCE_Constant, keys []}
      targetRotation=<rot 0 0 0>
     }
     {
      bApplyCustomCurve=false
      bIsHidden=false
      distanceMethod=DefaultMethod
      drivenName=Pose_2
      functionType=DefaultFunction
      targetScale=0.279
      boneTransforms [{targetRotation=<rot -37.3673 104.3609 -59.0518>, targetTranslation=<-9.8658 1.2136 0.4104>}]
      customCurve {defaultValue=3.4028234663852886e+38, postInfinityExtrap=RCCE_Constant, preInfinityExtrap=RCCE_Constant, keys []}
      targetRotation=<rot 0 0 0>
     }
     {
      bApplyCustomCurve=false
      bIsHidden=false
      distanceMethod=DefaultMethod
      drivenName=Pose_3
      functionType=DefaultFunction
      targetScale=0.7137
      boneTransforms [{targetRotation=<rot -5.8521 99.0705 -52.6784>, targetTranslation=<-9.8658 1.2136 0.4104>}]
      customCurve {defaultValue=3.4028234663852886e+38, postInfinityExtrap=RCCE_Constant, preInfinityExtrap=RCCE_Constant, keys []}
      targetRotation=<rot 0 0 0>
     }
     {
      bApplyCustomCurve=false
      bIsHidden=false
      distanceMethod=DefaultMethod
      drivenName=Pose_4
      functionType=DefaultFunction
      targetScale=0.5348
      boneTransforms [{targetRotation=<rot -50.9564 98.8603 -31.5024>, targetTranslation=<-9.8658 1.2136 0.4104>}]
      customCurve {defaultValue=3.4028234663852886e+38, postInfinityExtrap=RCCE_Constant, preInfinityExtrap=RCCE_Constant, keys []}
      targetRotation=<rot 0 0 0>
     }
     {
      bApplyCustomCurve=false
      bIsHidden=false
      distanceMethod=DefaultMethod
      drivenName=Pose_5
      functionType=DefaultFunction
      targetScale=1
      boneTransforms [{targetRotation=<rot 29.3585 90.2531 -81.8668>, targetTranslation=<-9.8658 1.2136 0.4104>}]
      customCurve {defaultValue=3.4028234663852886e+38, postInfinityExtrap=RCCE_Constant, preInfinityExtrap=RCCE_Constant, keys []}
      targetRotation=<rot 0 0 0>
     }
     {
      bApplyCustomCurve=false
      bIsHidden=false
      distanceMethod=DefaultMethod
      drivenName=Pose_6
      functionType=DefaultFunction
      targetScale=0.2583
      boneTransforms [{targetRotation=<rot -74.4506 67.5336 -20.2769>, targetTranslation=<-9.8658 1.2136 0.4104>}]
      customCurve {defaultValue=3.4028234663852886e+38, postInfinityExtrap=RCCE_Constant, preInfinityExtrap=RCCE_Constant, keys []}
      targetRotation=<rot 0 0 0>
     }
     {
      bApplyCustomCurve=false
      bIsHidden=false
      distanceMethod=DefaultMethod
      drivenName=Pose_7
      functionType=DefaultFunction
      targetScale=0.279
      boneTransforms [{targetRotation=<rot -46.6302 93.6139 -54.8469>, targetTranslation=<-9.8658 1.2136 0.4104>}]
      customCurve {defaultValue=3.4028234663852886e+38, postInfinityExtrap=RCCE_Constant, preInfinityExtrap=RCCE_Constant, keys []}
      targetRotation=<rot 0 0 0>
     }
     {
      bApplyCustomCurve=false
      bIsHidden=false
      distanceMethod=DefaultMethod
      drivenName=Pose_8
      functionType=DefaultFunction
      targetScale=0.4084
      boneTransforms [{targetRotation=<rot -66.3578 8.6543 -20.4238>, targetTranslation=<-9.8658 1.2136 0.4104>}]
      customCurve {defaultValue=3.4028234663852886e+38, postInfinityExtrap=RCCE_Constant, preInfinityExtrap=RCCE_Constant, keys []}
      targetRotation=<rot 0 0 0>
     }
    ]
    rBFParams {distanceMethod=Quaternion, function=Linear, radius=40}
    sourceBones [{boneName=upperarm_r}]
   }
   n1 = "Constraint - kneepad_r (PP)"(
    ComponentPose=Local To Component(LocalPose=Copy Pose From Mesh() {bUseAttachedParent=true})
    Alpha=1
    ConstraintWeights_0=0.4
    ConstraintWeights_1=0.6
   ) {
    alphaBoolBlend {blendOption=Linear}
    boneToModify {boneName=kneepad_r}
    constraintSetup [{targetBone {boneName=thigh_r}}, {targetBone {boneName=calf_r}}]
    constraintWeights [1,1]
   }
   n2 = "Transform (Modify) Bone - Bone: dyn_chains_01"(ComponentPose=@n1, Alpha=0, Translation="-0.160,0.000,-0.683", Rotation="0,0,0", Scale="1,1,1") {
    rotationMode=BMM_Additive
    rotationSpace=BCS_BoneSpace
    translationMode=BMM_Additive
    translationSpace=BCS_BoneSpace
    alphaBoolBlend {blendOption=Linear}
    boneToModify {boneName=dyn_chains_01}
    translation {x=-0.16, z=-0.683}
   }
   n3 = RigidBody(ComponentPose=@n2) {
    bClampLinearTranslationLimitToRefPose=true
    bOverrideWorldGravity=true
    overridePhysicsAsset=&/F_MED_Bandolier_tags_dyn
    simulationSpace=BaseBoneSpace
    alphaBoolBlend {blendOption=Linear}
    baseBoneRef {boneName=spine_05}
    componentAppliedLinearAccClamp=<1000 1000 1000>
    componentLinearAccScale=<0.2 0.2 0.3>
    componentLinearVelScale=<0.2 0.2 0.3>
    overrideWorldGravity {z=-980}
   }
   n4 = Anim Dynamics - Bone: dyn_arm_bandana_l(ComponentPose=@n3, Alpha=0.75, ExternalForce="0,0,0") {
    alpha=0.9
    angularSpringConstant=1500
    bAngularSpring=true
    bGravityOverrideInSimSpace=true
    numSolverIterationsPostUpdate=4
    numSolverIterationsPreUpdate=16
    simulationSpace=Actor
    alphaBoolBlend {blendOption=Linear}
    boundBone {boneName=dyn_arm_bandana_l}
    physicsBodyDefinitions [{
      sphereCollisionRadius=0
      boundBone {boneName=dyn_arm_bandana_l}
      boxExtents=<15 5 5>
      constraintSetup {angularTargetAxis=AxisZ, twistAxis=AxisY, angularLimitsMax=<15 0 15>, angularLimitsMin=<-15 0 -15>, angularTarget=<0 0 1>}
      localJointOffset=<0 0 0>
     }]
   }
   n5 = Look At - Bone: dyn_ammobelt_corrective(ComponentPose=@n4, Alpha=0.7) {
    alphaBoolBlend {blendOption=Linear}
    boneToModify {boneName=dyn_ammobelt_corrective}
    lookAtLocation=<-4 -2>
    lookAtTarget {boneReference {boneName=upperarm_r}}
    lookAt_Axis {axis=<1 0>}
    lookUp_Axis {axis=<0 1>}
   }
   n6 = "Transform (Modify) Bone - Bone: dyn_hip_magazine"(ComponentPose=@n5, Alpha=1, Translation="0,0,0", Rotation="-6,0,0", Scale="1,1,1") {
    rotationMode=BMM_Additive
    rotationSpace=BCS_ParentBoneSpace
    alphaBoolBlend {blendOption=Linear}
    boneToModify {boneName=dyn_hip_magazine}
    rotation {pitch=-6}
   }
   Output Pose(Result=Component To Local(ComponentPose=@n6)) {layerGroup=None}
  ]}
 EventGraph {nodes [
   Event BlueprintUpdateAnimation()
   n1 = BreakTransform(InTransform=GetDeltaTransformFromRefPose(self=GetPartSkeletalMeshComponent(), BoneName=thigh_l, BaseName=None))
   > Set thigh_l_alpha(thigh_l_alpha:double=MapRangeClamped(Value=BreakRotator(InRot=@n1.Rotation).Pitch, InRangeA=15, InRangeB=18)) -> (Output_Get:double)
   TryGetPawnOwner()
  ]}
]
variables [thigh_l_alpha {type=number}]
